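# Sound cue for the equip foley of the Figment Scavenger pistol, reusing the machine pistol reload-end wave.
# A fixed short delay feeds a modulator and then the single wave player.
# It plays in the weapon foley sound class with reload attenuation settings, and overrides concurrency to prevent new instances once the limit is reached.
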SoundCue /Figment_Pistol_Scavenger_Athena/Sounds/SC_Pistol_Scavenger_Athena_Figment_Equip_Cue
root=Delay_0
nodes [
 {id=Delay_0, children [Modulator_1], properties {delayMax=0.15, delayMin=0.15}}
 {id=Modulator_1, children [WavePlayer_15]}
 {id=WavePlayer_15, properties {soundWaveAssetPtr=/Game/Sounds/Fort_Weapons/pistol/pistol_foley/machine_pistol_reload_end}}
]
properties {
 attenuationSettings=&Fort_Weapon_Reload_Att
 bOverrideConcurrency=true
 maxDistance=2250
 soundClassObject=&Fort_WeaponFoley
 volumeMultiplier=1.002
 attenuationOverrides {falloffDistance=4100, attenuationShapeExtents {x=900}}
 concurrencyOverrides {resolutionRule=PreventNew}
}
