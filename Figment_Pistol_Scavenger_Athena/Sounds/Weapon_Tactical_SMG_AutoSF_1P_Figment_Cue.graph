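# First-person gunfire cue for the Figment Scavenger pistol, built from tactical SMG single-fire close samples.
# A mixer layers a randomly picked, modulated shot with a quieter low-ammo warning, which plays only when the LowAmmo branch parameter is true.
# The warning layer overrides its sound class to the critical low-ammo class.
# First-person attenuation with manual reverb send and gunfire concurrency govern playback.
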SoundCue /Figment_Pistol_Scavenger_Athena/Sounds/Weapon_Tactical_SMG_AutoSF_1P_Figment_Cue
root=Mixer_1
nodes [
 {id=Branch_3, children [WavePlayer_1,null,null], properties {boolParameterName=LowAmmo}}
 {id=Mixer_1, children [Modulator_0,SoundClass_0], properties {inputVolume [1,0.7]}}
 {id=Modulator_0, children [Random_0]}
 {id=Modulator_4, children [Branch_3], properties {pitchMax=1, volumeMax=1.15, volumeMin=1}}
 {id=Random_0, children [WavePlayer_4,WavePlayer_0,WavePlayer_2,WavePlayer_3], properties {weights [1,1,1,1]}}
 {
  id=SoundClass_0
  children [Modulator_4]
  properties {soundClassOverride=/Game/Sounds/FortSoundClassesAndModes/Fort_WeaponRanged_LowAmmo_Critical_PSM}
 }
 {id=WavePlayer_0, properties {soundWaveAssetPtr=&/sfx_tacsmg_singlefire_close_02}}
 {id=WavePlayer_1, properties {soundWaveAssetPtr=/Game/Sounds/Fort_Weapons/LowAmmo/weapon_low_ammo_hi_rifle_05}}
 {id=WavePlayer_2, properties {soundWaveAssetPtr=&/sfx_tacsmg_singlefire_close_03}}
 {id=WavePlayer_3, properties {soundWaveAssetPtr=&/sfx_tacsmg_singlefire_close_04}}
 {id=WavePlayer_4, properties {soundWaveAssetPtr=&/sfx_tacsmg_singlefire_close_01}}
]
properties {
 attenuationSettings=&Fort_Weapon_Projectile_FirstPerson_Athena_Att
 bDebug=true
 maxDistance=2097152
 soundClassObject=&Fort_WeaponRanged_Fire_1P_PSM
 volumeMultiplier=2
 attenuationOverrides {manualReverbSendLevel=-1.0165, reverbSendMethod=Manual}
 concurrencyOverrides {bLimitToOwner=true, maxCount=3, resolutionRule=StopOldest, volumeScale=0.9}
 concurrencySet [&Weapon_Gunfire_FirstPerson_Conc]
}
